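# First-person firing sound cue for the Clyde sniper rifle.
# A Random node picks between two equally weighted close-range sniper shot waves and feeds a Modulator that varies each shot, with the pitch floor held at normal.
# Output is boosted in volume and routed through first-person ranged-fire sound class, attenuation and gunfire concurrency assets.
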SoundCue /Clyde/Audio/Items/ClydeSniper/Weapon_RFL_Sniper_Clyde_1P_Cue
root=Modulator_17
nodes [
 {id=Modulator_17, children [Random_15], properties {pitchMin=1}}
 {id=Random_15, children [WavePlayer_33, WavePlayer_34], properties {bShouldExcludeFromBranchCulling=true, weights [1,1]}}
 {id=WavePlayer_33, properties {soundWaveAssetPtr=/Game/Sounds/Fort_Weapons/SniperRifles/Sniper_Rifle_Fire_Close_New_01}}
 {id=WavePlayer_34, properties {soundWaveAssetPtr=/Game/Sounds/Fort_Weapons/SniperRifles/Sniper_Rifle_Fire_Close_New_03}}
]
properties {
 attenuationSettings=&Fort_Weapon_Projectile_FirstPerson_Athena_Att
 maxDistance=2097152
 soundClassObject=&Fort_WeaponRanged_Fire_1P_PSM
 volumeMultiplier=1.5
 attenuationOverrides {distanceAlgorithm=NaturalSound, falloffDistance=4000, attenuationShapeExtents {x=1000}}
 concurrencyOverrides {bLimitToOwner=true, maxCount=3, resolutionRule=StopOldest}
 concurrencySet [&Weapon_Gunfire_FirstPerson_Conc]
}
